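# Translucent UI material that draws a challenge map marker: a round badge with a pointed tip and an icon.
# One exponential radial gradient is thresholded into container, content and icon-backing layers, while a sine-shaped mask adds the point.
# The layers and scaled, border-clamped icon texture drive color and opacity blends, with parameters for selection, hover, icon state and progress color.
# Locked desaturates the color, and global offset and scale adjust the UVs.
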Material /Game/Athena/UI/Challenges/Art/MapIcons/M_UI_MapMarker
graph {
 properties {
  bUsedWithStaticMesh=false
  blendMode=BLEND_Translucent
  materialDomain=MD_UI
  numCustomizedUVs=1
  pixelDepthOffsetMode=PDOM_Legacy
  usedShadingModels=Default Lit
 }
 nodes [
  n1 = LinearInterpolate(A=VectorParameter() {parameterName=RGB-1, defaultValue=<rgba 0.2016 0.8669 1 1>}.RGB, Alpha=ScalarParameter() {parameterName=Selected})
  n2 = LinearInterpolate(
   A=VectorParameter() {parameterName=RGB-2, defaultValue=<rgba 0.0024 0.0395 0.1329 1>}.RGB
   B=VectorParameter() {group=ProgressMeter, parameterName=ProgressBar-RGB-2, defaultValue=<rgba 0.0343 1 0.227 1>}.RGB
   Alpha=ScalarParameter() {group=ProgressMeter, parameterName=AnimateProgressBarColor, sliderMax=1}
  )
  n3 = ScalarParameter() {defaultValue=0.3, parameterName=Content-Offset}
  n4 = MaterialFunctionCall(
   CenterPosition=Constant2Vector() {g=0.4, r=0.5}
   Radius=ScalarParameter() {defaultValue=0.3, parameterName=Radius}
   Density=ScalarParameter() {defaultValue=4, parameterName=Density}
  ) {materialFunction=&RadialGradientExponential}
  n5 = Saturate(None=Multiply(A=Subtract(A=@n4.RadialGradientExponential, B=@n3), B=ScalarParameter() {defaultValue=8, parameterName=Content-Edge}))
  n6 = LinearInterpolate(
   A=LinearInterpolate(
    A=VectorParameter() {parameterName=RGB-3, defaultValue=<rgba 0.0091 0.0331 0.1812 1>}.RGB
    B=VectorParameter() {parameterName=RGB-3-Hovered, defaultValue=<rgba 0.0065 0.0802 0.5395 1>}.RGB
    Alpha=ScalarParameter() {parameterName=Selected}
   )
   Alpha=ScalarParameter() {parameterName=Hovered}
  )
  n7 = LinearInterpolate(
   A=VectorParameter() {parameterName=RGB-4, defaultValue=<rgba 0.8796 0.956 1 1>}.RGB
   B=VectorParameter() {parameterName=RGB-4-Hovered, defaultValue=<rgba 0 0.0392 0.1333 1>}.RGB
   Alpha=ScalarParameter() {parameterName=Icon-State}
  )
  n8 = ScalarParameter() {group=Icon, parameterName=X-Offset}
  n9 = ScalarParameter() {defaultValue=0.1, group=Icon, parameterName=Y-Offset}
  n10 = MaterialFunctionCall(
   UVs=Add(A=AppendVector(A=@n8, B=@n9), B=TextureCoordinate())
   Texture Scale=ScalarParameter() {defaultValue=0.3, group=Icon, parameterName=IconScale}
  ) {materialFunction=&ScaleUVsByCenter}
  n11 = Multiply(
   A=MaterialFunctionCall!(
    UVs=@n10.UVs
    Alpha=TextureSampleParameter2D(UVs=@n10.UVs) {parameterName=Icon, texture=/Game/Athena/UI/Challenges/Art/ChallengeTypeIcons/UI_ChallengeIcon_Collect}.A
   ) {materialFunction=/Game/UI/Foundation/Items/ClampAlphaToTransparentBorder}.Result
   B=ScalarParameter() {defaultValue=1, parameterName=Icon}
  )
  n12 = ScalarParameter() {defaultValue=0.4, parameterName=IconBacking-Offset}
  n13 = Saturate(None=Multiply(A=Subtract(A=@n4.RadialGradientExponential, B=@n12), B=ScalarParameter() {defaultValue=12, parameterName=IconBacking-Edge}))
  Desaturation(
   None=LinearInterpolate(A=LinearInterpolate(A=@n1, B=@n2, Alpha=@n5), B=LinearInterpolate(A=@n6, B=@n7, Alpha=@n11), Alpha=@n13)
   Fraction=ScalarParameter() {parameterName=bIsLocked}
  )
  n14 = ScalarParameter() {group=Icon, parameterName=Global-X-Offset}
  n15 = ScalarParameter() {defaultValue=-0.04, group=Icon, parameterName=Global-Y-Offset}
  MaterialFunctionCall(
   UVs=Add(A=TextureCoordinate(), B=AppendVector(A=@n14, B=@n15))
   Texture Scale=ScalarParameter() {defaultValue=1.35, parameterName=Global-UV-Scale}
  ) {materialFunction=&ScaleUVsByCenter}
  n16 = TextureCoordinate()
  n17 = Add(A=@n16, B=ScalarParameter() {defaultValue=-0.35, parameterName=Point-Mask-Y-Offset})
  n18 = ScalarParameter() {defaultValue=64, parameterName=Point-Edge-Multiply}
  n19 = Saturate(None=Multiply(A=ComponentMask(None=@n17) {g=true}, B=@n18))
  n20 = Abs(None=Add(A=@n16, B=ScalarParameter() {defaultValue=-0.5, parameterName=Point-X-Offset}))
  n21 = Sine(None=Add(A=ComponentMask(None=@n16) {g=true}, B=ScalarParameter() {defaultValue=0.09, parameterName=Point-Y-RangeAdd})) {period=2}
  n22 = ScalarParameter() {defaultValue=-2.27, parameterName=Point-Sine-PhaseOffset}
  n23 = Add(
   A=OneMinus(None=ComponentMask(None=@n20) {r=true})
   B=Multiply(A=Add(A=@n21, B=@n22), B=ScalarParameter() {defaultValue=0.3825, parameterName=Point-Sine-Amplitude})
  )
  n24 = ScalarParameter() {defaultValue=0.23, parameterName=Point-Edge-Offset}
  n25 = Saturate(None=Multiply(A=Subtract(A=@n23, B=@n24), B=@n18) {constB=24})
  n26 = Subtract(A=@n4.RadialGradientExponential, B=ScalarParameter() {defaultValue=0.04, parameterName=Container-Offset})
  n27 = ScalarParameter() {defaultValue=12, parameterName=Container-Edge}
  n28 = Add(A=Saturate(None=DotProduct(A=@n19, B=@n25)), B=Saturate(None=Multiply(A=@n26, B=@n27)))
  n29 = LinearInterpolate(
   A=Multiply(A=Saturate(None=@n28), B=ScalarParameter() {defaultValue=1, parameterName=Container-Edge-Alpha})
   B=ScalarParameter() {defaultValue=0.5, parameterName=Content-Alpha}
   Alpha=@n5
  )
  n30 = ScalarParameter() {defaultValue=0.85, parameterName=IconBacking-Alpha}
  Multiply(
   A=LinearInterpolate(A=LinearInterpolate(A=@n29, B=@n30, Alpha=@n13), B=ScalarParameter() {defaultValue=1, parameterName=Icon-Alpha}, Alpha=@n11)
   B=ScalarParameter() {defaultValue=1, parameterName=Alpha}
  )
 ]
}
