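# Translucent UI master material for challenge map markers, with parameter-driven selected, hovered and locked (desaturated) states.
# The default marker layers radial-gradient discs, a scalable icon texture with hero sizing, and an angular progress ring with color animation and optional delta segment.
# The bIsAllChallenges static switch swaps in a procedural all-reward badge, cut from rotated-UV edges and radial gradients.
# Shape masks drive both color and opacity, so only the icon uses a texture.
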Material /Game/Athena/UI/Challenges/Art/MapIcons/M_UI_MapMarker_Master
graph {
 properties {
  bUsedWithStaticMesh=false
  blendMode=BLEND_Translucent
  materialDomain=MD_UI
  pixelDepthOffsetMode=PDOM_Legacy
  usedShadingModels=Default Lit
 }
 nodes [
  n1 = VectorParameter() {parameterName=RGB-1, defaultValue=<rgba 0 0.7966 0.956 1>}
  n2 = ScalarParameter() {parameterName=Selected}
  n3 = VectorParameter() {parameterName=RGB-2, defaultValue=<rgba 0.0024 0.0395 0.1329 1>}
  n4 = VectorParameter() {group=ProgressMeter, parameterName=ProgressBar-RGB-2, defaultValue=<rgba 0.0343 1 0.227 1>}
  n5 = ScalarParameter() {group=ProgressMeter, parameterName=AnimateProgressBarColor, sliderMax=1}
  n6 = ScalarParameter() {defaultValue=0.4, parameterName=Content-Offset}
  n7 = MaterialFunctionCall(Radius=ScalarParameter() {defaultValue=0.5, parameterName=Radius}, Density=ScalarParameter() {defaultValue=4, parameterName=Density}) {materialFunction=&RadialGradientExponential}
  n8 = Saturate(None=Multiply(A=Subtract(A=@n7.RadialGradientExponential, B=@n6), B=ScalarParameter() {defaultValue=12, parameterName=Content-Edge}))
  n9 = LinearInterpolate(
   A=VectorParameter() {parameterName=RGB-3, defaultValue=<rgba 0.0065 0.0802 0.5395 1>}.RGB
   B=VectorParameter() {parameterName=RGB-3-Hovered, defaultValue=<rgba 0.0065 0.0802 0.5395 1>}.RGB
   Alpha=ScalarParameter() {parameterName=Selected}
  )
  n10 = ScalarParameter() {parameterName=Hovered}
  n11 = VectorParameter() {parameterName=RGB-4, defaultValue=<rgba 1 1 1 1>}
  n12 = VectorParameter() {parameterName=RGB-4-Hovered, defaultValue=<rgba 1 1 1 1>}
  n13 = ScalarParameter() {parameterName=Icon-State}
  n14 = AppendVector(A=ScalarParameter() {group=Icon, parameterName=X-Offset}, B=ScalarParameter() {group=Icon, parameterName=Y-Offset})
  n15 = TextureCoordinate()
  n16 = AppendVector(A=ScalarParameter() {group=Icon, parameterName=X-Offset}, B=ScalarParameter() {group=Icon, parameterName=Y-Offset})
  n17 = LinearInterpolate(
   A=MaterialFunctionCall(UVs=Add(A=@n14, B=@n15), Texture Scale=ScalarParameter() {defaultValue=0.425, group=Icon, parameterName=IconScale}) {materialFunction=&ScaleUVsByCenter}.UVs
   B=MaterialFunctionCall(UVs=Add(A=@n15, B=@n16), Texture Scale=ScalarParameter() {defaultValue=0.5, group=Icon, parameterName=HeroScale}) {materialFunction=&ScaleUVsByCenter}.UVs
   Alpha=ScalarParameter() {parameterName=bIsHero}
  )
  n18 = Multiply(
   A=MaterialFunctionCall!(
    UVs=@n17
    Alpha=TextureSampleParameter2D(UVs=@n17) {parameterName=Icon, texture=/Game/Athena/UI/Challenges/Art/ChallengeTypeIcons/UI_ChallengeIcon_Collect}.A
   ) {materialFunction=/Game/UI/Foundation/Items/ClampAlphaToTransparentBorder}.Result
   B=ScalarParameter() {defaultValue=1, parameterName=Icon}
  )
  n19 = ScalarParameter() {defaultValue=0.7524, parameterName=IconBacking-Offset}
  n20 = Saturate(None=Multiply(A=Subtract(A=@n7.RadialGradientExponential, B=@n19), B=ScalarParameter() {defaultValue=12, parameterName=IconBacking-Edge}))
  n21 = LinearInterpolate(
   A=LinearInterpolate(A=LinearInterpolate(A=@n1.RGB, Alpha=@n2), B=LinearInterpolate(A=@n3.RGB, B=@n4.RGB, Alpha=@n5), Alpha=@n8)
   B=LinearInterpolate(A=LinearInterpolate(A=@n9, Alpha=@n10), B=LinearInterpolate(A=@n11.RGB, B=@n12.RGB, Alpha=@n13), Alpha=@n18)
   Alpha=@n20
  )
  n22 = VectorParameter() {group=ProgressMeter, parameterName=ProgressBar-RGB, defaultValue=<rgba 0.956 0.8305 0 1>}
  n23 = Subtract(A=TextureCoordinate()) {constB=0.5}
  n24 = ScalarParameter() {defaultValue=0.25, group=ProgressMeter, parameterName=VectorRad-Rotate}
  n25 = OneMinus(None=MaterialFunctionCall(Vector or UVs=Rotator(Coordinate=@n23, Time=@n24) {centerX=0, centerY=0, speed=6.2832}) {materialFunction=&VectorToRadialValue}.Vector Converted to Angle)
  n26 = ScalarParameter() {defaultValue=0.25, group=ProgressMeter, parameterName=Progress}
  n27 = LinearInterpolate(Alpha=LinearInterpolate(
    A=@n26
    B=ScalarParameter() {defaultValue=0.5, group=ProgressMeter, parameterName=DeltaProgress}
    Alpha=ScalarParameter() {group=ProgressMeter, parameterName=AnimateDeltaProgress, sliderMax=1}
   )) {constA=1, constB=0}
  n28 = ScalarParameter() {defaultValue=128, group=ProgressMeter, parameterName=Progress-Edge}
  n29 = Saturate(None=Multiply(A=Subtract(A=@n25, B=@n27), B=@n28))
  n30 = LinearInterpolate(
   A=@n22.RGB
   B=VectorParameter() {group=ProgressMeter, parameterName=DeltaProgressBar-RGB, defaultValue=<rgba 0.956 0 0.7798 1>}.RGB
   Alpha=@n29
  )
  n31 = LinearInterpolate(Alpha=@n26) {constA=1, constB=0}
  n32 = Saturate(None=Multiply(A=Subtract(A=@n25, B=@n31), B=@n28))
  n33 = LinearInterpolate(
   A=StaticSwitchParameter(True=LinearInterpolate(A=@n30, B=@n22.RGB, Alpha=@n32), False=@n22.RGB) {parameterName=ShowDeltaProgress}
   B=VectorParameter() {group=ProgressMeter, parameterName=ProgressBar-RGB-2, defaultValue=<rgba 0.0343 1 0.227 1>}.RGB
   Alpha=ScalarParameter() {group=ProgressMeter, parameterName=AnimateProgressBarColor, sliderMax=1}
  )
  n34 = LinearInterpolate(
   B=StaticSwitchParameter(True=LinearInterpolate(A=@n29, Alpha=@n32), False=@n32) {parameterName=ShowDeltaProgress}
   Alpha=Multiply(A=Subtract(A=@n8, B=@n20), B=ScalarParameter() {defaultValue=1, parameterName=Progress-Alpha})
  )
  n35 = VectorParameter() {parameterName=RGB-1, defaultValue=<rgba 0 0.7966 0.956 1>}
  n36 = ScalarParameter() {parameterName=Selected}
  n37 = TextureCoordinate()
  n38 = ScalarParameter() {defaultValue=0.5, group=All-Reward-Shape, parameterName=Right-Angel}
  n39 = Subtract(
   A=ComponentMask(None=Rotator(Coordinate=@n37, Time=@n38) {speed=6.2832}) {r=true}
   B=ScalarParameter() {defaultValue=0.25, group=All-Reward-Shape, parameterName=Right-Inner-OffSet-S1}
  )
  n40 = ScalarParameter() {defaultValue=64, group=All-Reward-Shape, parameterName=Right-Edge}
  n41 = TextureCoordinate()
  n42 = ScalarParameter() {group=All-Reward-Shape, parameterName=Left-Angel}
  n43 = Subtract(
   A=ComponentMask(None=Rotator(Coordinate=@n41, Time=@n42) {speed=6.2832}) {r=true}
   B=ScalarParameter() {defaultValue=0.25, group=All-Reward-Shape, parameterName=Left-Inner-OffSet-S1}
  )
  n44 = ScalarParameter() {defaultValue=64, group=All-Reward-Shape, parameterName=Left-Edge}
  n45 = DotProduct(A=Saturate(None=Multiply(A=@n39, B=@n40)), B=Saturate(None=Multiply(A=@n43, B=@n44)))
  n46 = ComponentMask(None=Rotator(Coordinate=TextureCoordinate(), Time=ScalarParameter() {group=All-Reward-Shape, parameterName=Top-Angel}) {speed=6.2832}) {g=true}
  n47 = ScalarParameter() {defaultValue=0.25, group=All-Reward-Shape, parameterName=Top-Inner-OffSet-S1}
  n48 = Saturate(None=Multiply(A=Subtract(A=@n46, B=@n47), B=ScalarParameter() {defaultValue=64, group=All-Reward-Shape, parameterName=Top-Edge}))
  n49 = ScalarParameter() {defaultValue=0.475, group=All-Reward-Shape, parameterName=Bot-Radial-U}
  n50 = ScalarParameter() {defaultValue=-0.25, group=All-Reward-Shape, parameterName=Bot-Radial-V}
  n51 = Multiply(
   A=MaterialFunctionCall(
    CenterPosition=AppendVector(A=@n49, B=@n50)
    Radius=ScalarParameter() {defaultValue=1, group=All-Reward-Shape, parameterName=Bot-Radius-S1}
    Density=ScalarParameter() {defaultValue=4, group=All-Reward-Shape, parameterName=Bot-Density}
   ) {materialFunction=&RadialGradientExponential}.RadialGradientExponential
   B=ScalarParameter() {defaultValue=96, group=All-Reward-Shape, parameterName=Bot-Inner-Edge}
  )
  n52 = Saturate(None=DotProduct(A=DotProduct(A=@n48, B=@n45), B=Saturate(None=@n51)))
  n53 = LinearInterpolate(
   A=VectorParameter() {parameterName=RGB-3, defaultValue=<rgba 0.0065 0.0802 0.5395 1>}.RGB
   B=VectorParameter() {parameterName=RGB-3-Hovered, defaultValue=<rgba 0.0065 0.0802 0.5395 1>}.RGB
   Alpha=ScalarParameter() {parameterName=Selected}
  )
  n54 = ScalarParameter() {parameterName=Hovered}
  n55 = VectorParameter() {parameterName=RGB-4, defaultValue=<rgba 1 1 1 1>}
  n56 = VectorParameter() {parameterName=RGB-4-Hovered, defaultValue=<rgba 1 1 1 1>}
  n57 = ScalarParameter() {parameterName=Icon-State}
  n58 = AppendVector(A=ScalarParameter() {group=Icon, parameterName=X-Offset}, B=ScalarParameter() {group=Icon, parameterName=Y-Offset})
  n59 = TextureCoordinate()
  n60 = AppendVector(A=ScalarParameter() {group=Icon, parameterName=X-Offset}, B=ScalarParameter() {group=Icon, parameterName=Y-Offset})
  n61 = LinearInterpolate(
   A=MaterialFunctionCall(UVs=Add(A=@n58, B=@n59), Texture Scale=ScalarParameter() {defaultValue=0.425, group=Icon, parameterName=IconScale}) {materialFunction=&ScaleUVsByCenter}.UVs
   B=MaterialFunctionCall(UVs=Add(A=@n59, B=@n60), Texture Scale=ScalarParameter() {defaultValue=0.5, group=Icon, parameterName=HeroScale}) {materialFunction=&ScaleUVsByCenter}.UVs
   Alpha=ScalarParameter() {parameterName=bIsHero}
  )
  n62 = Multiply(
   A=MaterialFunctionCall(
    UVs=@n61
    Alpha=TextureSampleParameter2D(UVs=@n61) {parameterName=Icon, texture=&UI_ChallengeIcon_Collect}.A
   ) {materialFunction=&ClampAlphaToTransparentBorder}.Result
   B=ScalarParameter() {defaultValue=1, parameterName=Icon}
  )
  n63 = TextureCoordinate()
  n64 = ScalarParameter() {defaultValue=0.5, group=All-Reward-Shape, parameterName=Right-Angel}
  n65 = Subtract(
   A=ComponentMask(None=Rotator(Coordinate=@n63, Time=@n64) {speed=6.2832}) {r=true}
   B=ScalarParameter() {defaultValue=0.25, group=All-Reward-Shape, parameterName=Right-Inner-OffSet-S2}
  )
  n66 = ScalarParameter() {defaultValue=64, group=All-Reward-Shape, parameterName=Right-Edge}
  n67 = TextureCoordinate()
  n68 = ScalarParameter() {group=All-Reward-Shape, parameterName=Left-Angel}
  n69 = Subtract(
   A=ComponentMask(None=Rotator(Coordinate=@n67, Time=@n68) {speed=6.2832}) {r=true}
   B=ScalarParameter() {defaultValue=0.25, group=All-Reward-Shape, parameterName=Left-Inner-OffSet-S2}
  )
  n70 = ScalarParameter() {defaultValue=64, group=All-Reward-Shape, parameterName=Left-Edge}
  n71 = DotProduct(A=Saturate(None=Multiply(A=@n65, B=@n66)), B=Saturate(None=Multiply(A=@n69, B=@n70)))
  n72 = ComponentMask(None=Rotator(Coordinate=TextureCoordinate(), Time=ScalarParameter() {group=All-Reward-Shape, parameterName=Top-Angel}) {speed=6.2832}) {g=true}
  n73 = ScalarParameter() {defaultValue=0.25, group=All-Reward-Shape, parameterName=Top-Inner-OffSet-S2}
  n74 = Saturate(None=Multiply(A=Subtract(A=@n72, B=@n73), B=ScalarParameter() {defaultValue=64, group=All-Reward-Shape, parameterName=Top-Edge}))
  n75 = ScalarParameter() {defaultValue=0.475, group=All-Reward-Shape, parameterName=Bot-Radial-U}
  n76 = ScalarParameter() {defaultValue=-0.25, group=All-Reward-Shape, parameterName=Bot-Radial-V}
  n77 = Multiply(
   A=MaterialFunctionCall(
    CenterPosition=AppendVector(A=@n75, B=@n76)
    Radius=ScalarParameter() {defaultValue=1, group=All-Reward-Shape, parameterName=Bot-Radius-S2}
    Density=ScalarParameter() {defaultValue=4, group=All-Reward-Shape, parameterName=Bot-Density}
   ) {materialFunction=&RadialGradientExponential}.RadialGradientExponential
   B=ScalarParameter() {defaultValue=96, group=All-Reward-Shape, parameterName=Bot-Inner-Edge}
  )
  n78 = Saturate(None=DotProduct(A=DotProduct(A=@n74, B=@n71), B=Saturate(None=@n77)))
  n79 = LinearInterpolate(
   A=LinearInterpolate(
    A=LinearInterpolate(A=@n35.RGB, Alpha=@n36)
    B=VectorParameter() {parameterName=RGB-2, defaultValue=<rgba 0.0024 0.0395 0.1329 1>}.RGB
    Alpha=@n52
   )
   B=LinearInterpolate(A=LinearInterpolate(A=@n53, Alpha=@n54), B=LinearInterpolate(A=@n55.RGB, B=@n56.RGB, Alpha=@n57), Alpha=@n62)
   Alpha=@n78
  )
  n80 = ScalarParameter() {parameterName=bIsLocked}
  Desaturation(
   None=StaticSwitchParameter(True=Desaturation(None=@n79, Fraction=@n80), False=LinearInterpolate(A=@n21, B=@n33, Alpha=@n34)) {parameterName=bIsAllChallenges}
   Fraction=ScalarParameter() {parameterName=bIsLocked}
  )
  n81 = Rotator(Coordinate=TextureCoordinate(), Time=ScalarParameter() {group=All-Reward-Shape, parameterName=Top-Angel}) {speed=6.2832}
  n82 = Multiply(
   A=Subtract(A=ComponentMask(None=@n81) {g=true}, B=ScalarParameter() {defaultValue=0.0125, group=All-Reward-Shape, parameterName=Top-OffSet})
   B=ScalarParameter() {defaultValue=64, group=All-Reward-Shape, parameterName=Top-Edge}
  )
  n83 = ComponentMask(None=Rotator(Coordinate=TextureCoordinate(), Time=ScalarParameter() {defaultValue=0.5, group=All-Reward-Shape, parameterName=Right-Angel}) {speed=6.2832}) {r=true}
  n84 = ScalarParameter() {defaultValue=0.0125, group=All-Reward-Shape, parameterName=Right-OffSet}
  n85 = Saturate(None=Multiply(A=Subtract(A=@n83, B=@n84), B=ScalarParameter() {defaultValue=64, group=All-Reward-Shape, parameterName=Right-Edge}))
  n86 = ComponentMask(None=Rotator(Coordinate=TextureCoordinate(), Time=ScalarParameter() {group=All-Reward-Shape, parameterName=Left-Angel}) {speed=6.2832}) {r=true}
  n87 = ScalarParameter() {defaultValue=0.0125, group=All-Reward-Shape, parameterName=Left-OffSet}
  n88 = Saturate(None=Multiply(A=Subtract(A=@n86, B=@n87), B=ScalarParameter() {defaultValue=64, group=All-Reward-Shape, parameterName=Left-Edge}))
  n89 = Rotator(Coordinate=TextureCoordinate(), Time=ScalarParameter() {defaultValue=0.375, group=All-Reward-Shape, parameterName=BotRight-Angel}) {speed=6.2832}
  n90 = Multiply(
   A=Subtract(A=ComponentMask(None=@n89) {r=true}, B=ScalarParameter() {defaultValue=0.0125, group=All-Reward-Shape, parameterName=BotRight-OffSet})
   B=ScalarParameter() {defaultValue=64, group=All-Reward-Shape, parameterName=BotRight-Edge}
  )
  n91 = Rotator(Coordinate=TextureCoordinate(), Time=ScalarParameter() {defaultValue=0.125, group=All-Reward-Shape, parameterName=BotLeft-Angel}) {speed=6.2832}
  n92 = Multiply(
   A=Subtract(A=ComponentMask(None=@n91) {r=true}, B=ScalarParameter() {defaultValue=0.0125, group=All-Reward-Shape, parameterName=BotLeft-OffSet})
   B=ScalarParameter() {defaultValue=64, group=All-Reward-Shape, parameterName=BotLeft-Edge}
  )
  n93 = DotProduct(A=DotProduct(A=Saturate(None=@n82), B=DotProduct(A=@n85, B=@n88)), B=DotProduct(A=Saturate(None=@n90), B=Saturate(None=@n92)))
  n94 = ScalarParameter() {defaultValue=0.75, group=All-Reward-Shape, parameterName=Inner-Value-S1}
  n95 = Subtract(
   A=Subtract(A=Saturate(None=@n93), B=Multiply(A=@n52, B=@n94))
   B=Multiply(A=@n78, B=ScalarParameter() {group=All-Reward-Shape, parameterName=Inner-Value-S2})
  )
  n96 = Subtract(A=@n7.RadialGradientExponential, B=ScalarParameter() {defaultValue=0.1, parameterName=Container-Offset})
  n97 = ScalarParameter() {defaultValue=12, parameterName=Container-Edge}
  n98 = Multiply(A=Saturate(None=Multiply(A=@n96, B=@n97)), B=ScalarParameter() {defaultValue=1, parameterName=Container-Edge-Alpha})
  n99 = ScalarParameter() {defaultValue=0.5, parameterName=Content-Alpha}
  n100 = LinearInterpolate(
   A=LinearInterpolate(A=LinearInterpolate(A=@n98, B=@n99, Alpha=@n8), B=ScalarParameter() {defaultValue=0.75, parameterName=IconBacking-Alpha}, Alpha=@n20)
   B=ScalarParameter() {defaultValue=1, parameterName=Icon-Alpha}
   Alpha=@n18
  )
  Multiply(
   A=StaticSwitchParameter(True=Add(A=@n62, B=@n95), False=Add(A=@n34, B=@n100)) {parameterName=bIsAllChallenges}
   B=ScalarParameter() {defaultValue=1, parameterName=Alpha}
  )
 ]
}
