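# UI-domain translucent master material for challenge map marker states, drawing a coloured procedural shape with an optional icon.
# Four rotated UV gradients with per-side angle, offset and edge sharpness are multiplied via scalar dot products to form the shape mask.
# When bHasIcon is on, a centre-scaled, translated icon texture's alpha blends opacity toward IconAlpha and colour from OuterRing to InnerFill.
# Overall_Alpha scales the final opacity.
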Material /Game/Athena/UI/Challenges/Art/MapIcons/M_UI_MarkerState_Master
graph {
 properties {
  bUsedWithStaticMesh=false
  blendMode=BLEND_Translucent
  materialDomain=MD_UI
  numCustomizedUVs=1
  pixelDepthOffsetMode=PDOM_Legacy
  usedShadingModels=Default Lit
 }
 nodes [
  n1 = TextureCoordinate()
  n2 = AppendVector(
   A=ScalarParameter() {defaultValue=0.5, group=Icon, parameterName=X-ScaleIcon}
   B=ScalarParameter() {defaultValue=0.5, group=Icon, parameterName=Y-ScaleIcon}
  )
  n3 = ScalarParameter() {group=Icon, parameterName=Translate_Icon_X}
  n4 = ScalarParameter() {group=Icon, parameterName=Translate_Icon_Y}
  n5 = TextureSampleParameter2D(UVs=Add(
    A=MaterialFunctionCall(UVs=@n1, Texture Scale=@n2) {materialFunction=&ScaleUVsByCenter}.UVs
    B=AppendVector(A=@n3, B=@n4)
   )) {
   group=Icon
   parameterName=ContextState_Icon
   samplerSource=SSM_Clamp_WorldGroupSettings
   texture=/Game/UI/Foundation/Textures/Icons/Checks/T-Icon-Check-32
  }
  n6 = TextureCoordinate()
  n7 = ComponentMask(None=Rotator(Coordinate=@n6, Time=ScalarParameter() {parameterName=L-Angle}) {speed=6.2832}) {r=true}
  n8 = ScalarParameter() {parameterName=L-Offset}
  n9 = Saturate(None=Multiply(A=Subtract(A=@n7, B=@n8), B=ScalarParameter() {parameterName=L-Edge}))
  n10 = ComponentMask(None=Rotator(Coordinate=@n6, Time=ScalarParameter() {parameterName=R-Angle}) {speed=6.2832}) {r=true}
  n11 = ScalarParameter() {parameterName=R-Offset}
  n12 = Saturate(None=Multiply(A=Subtract(A=@n10, B=@n11), B=ScalarParameter() {parameterName=R-Edge}))
  n13 = TextureCoordinate()
  n14 = ComponentMask(None=Rotator(Coordinate=@n13, Time=ScalarParameter() {parameterName=T-Angle}) {speed=6.2832}) {g=true}
  n15 = ScalarParameter() {parameterName=T-Offset}
  n16 = Saturate(None=Multiply(A=Subtract(A=@n14, B=@n15), B=ScalarParameter() {parameterName=T-Edge}))
  n17 = ComponentMask(None=Rotator(Coordinate=@n13, Time=ScalarParameter() {parameterName=B-Angle}) {speed=6.2832}) {g=true}
  n18 = ScalarParameter() {parameterName=B-Offset}
  n19 = Saturate(None=Multiply(A=Subtract(A=@n17, B=@n18), B=ScalarParameter() {parameterName=B-Edge}))
  n20 = Saturate(None=DotProduct(A=DotProduct(A=@n9, B=@n12), B=DotProduct(A=@n16, B=@n19)))
  n21 = StaticSwitchParameter(True=LinearInterpolate(A=@n20, B=ScalarParameter() {parameterName=IconAlpha}, Alpha=@n5.A), False=@n20) {parameterName=bHasIcon}
  Saturate(None=Multiply(A=Saturate(None=@n21), B=ScalarParameter() {defaultValue=1, parameterName=Overall_Alpha}))
  n22 = VectorParameter() {group=OuterRing, parameterName=OuterRing_Color, defaultValue=<rgba 0 0.6667 1 1>}
  StaticSwitchParameter(
   True=LinearInterpolate(
    A=@n22.RGB
    B=VectorParameter() {group=InnerFill, parameterName=InnerFill_Color, defaultValue=<rgba 0 0.1667 1 1>}.RGB
    Alpha=Saturate(None=@n5.A)
   )
   False=@n22.RGB
  ) {parameterName=bHasIcon}
 ]
}
